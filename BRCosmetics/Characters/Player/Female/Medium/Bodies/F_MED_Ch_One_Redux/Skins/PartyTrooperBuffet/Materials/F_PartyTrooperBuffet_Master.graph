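# Master character material for the female PartyTrooperBuffet skin body, adding sequins and faux-depth diamonds between the shared character PreFX and PostFX functions.
# Fxmap R masks two-tone tinted sequins from MF_BuffetSequins, then metallic minus G and B masks MF_MetalsLUT.
# Fxmap B blends in clear-coat diamonds, using refracted-view bump offsets for interior parallax and a LUT for emissive rainbow chips.
# CINE_MF_CharPostFX finishes the chain.
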Material /BRCosmetics/Characters/Player/Female/Medium/Bodies/F_MED_Ch_One_Redux/Skins/PartyTrooperBuffet/Materials/F_PartyTrooperBuffet_Master
graph {
 properties {
  bUseMaterialAttributes=true
  bUsedWithMeshDeformer=true
  bUsedWithMorphTargets=true
  bUsedWithSkeletalMesh=true
  pixelDepthOffsetMode=PDOM_Legacy
  shadingModel=MSM_FromMaterialExpression
  usedShadingModels=Default Lit | Clear Coat | Subsurface Profile
 }
 nodes [
  n1 = MaterialFunctionCall() {materialFunction=&BaseFN_CharacterFunction_PreFX}
  n2 = GetMaterialAttributes(None=@n1.Result) {attributeGetTypes [#1, #2]}
  n3 = VectorParameter() {parameterName=SequinColorInner, defaultValue=<rgba 0.6663 0.575 1 1>}
  n4 = TextureSampleParameter2D() {
   parameterName=Fxmap
   texture=/BRCosmetics/Characters/Player/Female/Medium/Bodies/F_MED_Ch_One_Redux/Skins/PartyTrooperBuffet/Textures/T_F_MED_Party_Trooper_Buffet_Body_FX
  }
  n5 = VectorParameter() {parameterName=SequinColorOuter, defaultValue=<rgba 0.9546 0.8542 1 1>}
  n6 = GetMaterialAttributes(None=MaterialFunctionCall!(
    SequinColorInner=Multiply(A=@n2.Base Color, B=@n3.RGB)
    Fresnel=NamedRerouteUsage()
    inMat=@n2.MaterialAttributes
    Mask=@n4.R
    SequinColorOuter=Multiply(A=@n5.RGB, B=@n2.Base Color)
   ) {materialFunction=/Game/Gadgets/Assets/VinderTech_Umbrella/Umbrella_Buffet/Materials/MF_BuffetSequins}.Result) {attributeGetTypes [#2, #3]}
  n7 = Subtract(A=Subtract(A=@n6.Metallic, B=@n4.G), B=@n4.B)
  n8 = GetMaterialAttributes(None=MaterialFunctionCall(Mask=Saturate(None=@n7), Materials=@n6.MaterialAttributes) {materialFunction=&MF_MetalsLUT}.Result) {attributeGetTypes [#2, #3, #1]}
  n9 = Constant()
  n10 = TextureCoordinate() {uTiling=8, vTiling=8}
  n11 = Constant() {r=1}
  n12 = MaterialFunctionCall!(
   Ray Direction=CameraVectorWS()
   Surface Normal=PixelNormalWS()
   Refractive Index Origin
   Refractive Index Target=ScalarParameter() {defaultValue=2, parameterName=ref}
  ) {materialFunction=/Engine/Functions/Engine_MaterialFunctions01/Vectors/Refract}
  n13 = TextureSample(UVs=MaterialFunctionCall!(
    HeightTexture=@n9
    TextureCoordinates=@n10
    ReferencePlane=@n11
    HeightRatioInput=ScalarParameter() {defaultValue=0.5, parameterName=rat}
    CameraVector=@n12.Vector
   ) {materialFunction=/Engine/Functions/Engine_MaterialFunctions02/Utility/BumpOffset_Advanced.BumpOffset_advanced}.UVs) {samplerSource=SSM_Wrap_WorldGroupSettings, texture=/Game/Gadgets/Assets/VinderTech_Umbrella/Umbrella_Buffet/Materials/T_SequinDiamond_BG}
  n14 = ScalarParameter() {defaultValue=10, parameterName=diamondBGBrightness}
  n15 = TextureSample(UVs=MaterialFunctionCall(
    HeightTexture=@n9
    TextureCoordinates=@n10
    ReferencePlane=@n11
    HeightRatioInput=ScalarParameter() {defaultValue=0.5, parameterName=rat2}
    CameraVector=@n12.Vector
   ) {materialFunction=&BumpOffset_advanced}.UVs) {samplerSource=SSM_Wrap_WorldGroupSettings, texture=&T_SequinDiamond_BG}
  n16 = OneMinus(None=NamedRerouteUsage())
  n17 = Multiply(
   A=@n15.A
   B=TextureSample(UVs=@n16) {texture=/Game/Gadgets/Assets/VinderTech_Umbrella/Umbrella_Buffet/Materials/T_SequinDiamond_RainbowSparkle_LUT}.RGB
  )
  n18 = ScalarParameter() {defaultValue=3, parameterName=ChomaticChipBrightness}
  n19 = SetMaterialAttributes!(
   MaterialAttributes=@n8.MaterialAttributes
   Shading Model=ShadingModel() {shadingModel=MSM_ClearCoat}
   Clear Coat or Curvature or Custom Data 0=Constant() {r=1}
   Clear Coat Roughness or Custom Data 1=Constant() {r=0.15}
   Base Color=Multiply(A=NamedRerouteUsage(), B=Multiply(A=@n13.RGB, B=@n14))
   Roughness=ScalarParameter() {defaultValue=0.2, parameterName=InnerDiamondRoughness}
   Metallic=ScalarParameter() {defaultValue=1, parameterName=innerdiamondMetallic}
   Emissive Color=Multiply(A=NamedRerouteUsage(), B=Multiply(A=@n17, B=@n18)) {constB=0}
  ) {attributeSetTypes [
    #4
    #5
    #6
    #2
    #7
    #3
    #8
   ]}
  n20 = MaterialFunctionCall(
   AO=@n1.AO
   Distance Blend=@n1.AO
   Fresnel=@n1.AO
   Gmap=@n1.AO
   Pre Skinned Local Position=@n1.AO
   Material=BlendMaterialAttributes(A=@n8.MaterialAttributes, B=@n19, Alpha=@n4.B)
  ) {materialFunction=&BaseFN_CharacterFunction_PostFX}
  MaterialFunctionCall(In=@n20.Result) {materialFunction=&CINE_MF_CharPostFX}
  NamedRerouteDeclaration(None=@n1.Fresnel) {name=existingFresnel, nodeColor=<rgba 0.4766 1 0 1>}
  NamedRerouteDeclaration(None=Power(Base=@n16) {constExponent=4}) {name=highContrastFacingAngle, nodeColor=<rgba 0 0.0859 1 1>}
 ]
}
